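# Animation Blueprint for the green jacket backpack cosmetic that adds secondary jiggle to its handle bone.
# The AnimGraph feeds the reference pose through a single low-alpha Anim Dynamics node on dyn_handle and outputs the result.
# The node simulates a box body with linear and angular springs, overridden damping, reduced gravity and tight linear and angular limits.
# The EventGraph holds only unconnected stub nodes, and no variables are defined.
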AnimBlueprint /BRCosmetics/Accessories/FORT_Backpacks/Backpack_F_MED_Green_Jacket/Meshes/F_MED_Green_Jacket_Pack_AnimBP
target_skeleton=&/F_MED_Green_Jacket_Pack_Skeleton
graphs [
 AnimGraph {nodes [
   n1 = Anim Dynamics - Bone: dyn_handle(ComponentPose=Local To Component(LocalPose=Local Space Ref Pose()), Alpha=0.1, ExternalForce="0,0,0") {
    angularDampingOverride=1
    angularSpringConstant=15
    bAngularSpring=true
    bGravityOverrideInSimSpace=true
    bLinearSpring=true
    bOverrideAngularDamping=true
    bOverrideLinearDamping=true
    gravityScale=0.6
    linearDampingOverride=1
    linearSpringConstant=250
    alphaBoolBlend {blendOption=Linear}
    boundBone {boneName=dyn_handle}
    componentAppliedLinearAccClamp=<10000 10000 10000>
    componentLinearAccScale=<0.5 0.5 0.5>
    componentLinearVelScale {z=0.5}
    gravityOverride {z=-980}
    physicsBodyDefinitions [{
      sphereCollisionRadius=0
      boundBone {boneName=dyn_handle}
      boxExtents=<15 15 5>
      constraintSetup {angularLimitsMax=<45 45 45>, angularLimitsMin=<-45 -45 -45>, angularTarget=<0 0 0>, linearAxesMax=<0.5 0.5 5>, linearAxesMin=<-0.5 -0.5 -1>}
      localJointOffset=<0 0 -3>
     }]
   }
   Output Pose(Result=Component To Local(ComponentPose=@n1)) {layerGroup=None}
  ]}
 EventGraph {nodes [
   TryGetPawnOwner()
   Event BlueprintUpdateAnimation()
  ]}
]
variables []
